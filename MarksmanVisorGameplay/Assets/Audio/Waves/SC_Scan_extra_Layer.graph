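# Sound cue that plays the extra layer of the Marksman Visor's scan audio.
# A single wave player feeds the Scan_extra_Layer wave straight to the output, with no mixing or randomization.
# It uses the visor's distant enemy-scan attenuation, routes to the Fort_Gadgets sound class, and overrides concurrency to cap simultaneous instances.
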SoundCue /MarksmanVisorGameplay/Assets/Audio/Waves/SC_Scan_extra_Layer
root=WavePlayer_0
nodes [{id=WavePlayer_0, properties {soundWaveAssetPtr=&/Scan_extra_Layer}}]
properties {
 attenuationSettings=/MarksmanVisorGameplay/Assets/Audio/ATT_Marksman_Visor_Scan_Enemy_Distant
 bOverrideConcurrency=true
 maxDistance=21000
 soundClassObject=&Fort_Gadgets
 volumeMultiplier=1
 concurrencyOverrides {maxCount=4}
}
